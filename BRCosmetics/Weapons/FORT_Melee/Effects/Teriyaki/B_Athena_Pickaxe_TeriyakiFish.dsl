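; Cosmetic weapon Blueprint for the Teriyaki Fish pickaxe in Battle Royale, handling its effects setup.
; On weapon cosmetics init it calls the parent, then nudges the idle particle system component upward with a relative location offset.
; The offset keeps that component off the origin so Spawn Per Unit emission triggers properly in the frontend.
; The construction script only calls the parent.
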(event EventOnInitWeaponCosmetics
  ; Added Offset for PSC to trigger the "Spawn per Unit". Otherwise the Location stay 0,0,0 and it doesnt trigger properly in FE
  (|CallParentFunction)
  (Transformation|AddRelativeLocation (CallFunction|GetIdlePSC) "0.000000,0.000000,20.000000" false false))

(fn ConstructionScript
  (|CallParentFunction))
